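# Editor utility Blueprint in the anisotropic hair-bake tools that snaps the level viewport camera to a fixed location and rotation.
# Its construction script starts a one-shot 0.1-second timer with a created event delegate, presumably bound to the SetCamera custom event.
# SetCamera calls SetLevelViewportCameraInfo on the Unreal Editor Subsystem with hard-coded transform values.
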EditorUtilityBlueprint /Game/NeverCook/ArtTools/AnisotropicBaker/HairBake/BP_HairBake_CameraPosition
graphs [
 EventGraph {nodes [
   SetCamera() {is=title K2Node_CustomEvent}
   > SetLevelViewportCameraInfo!(
    execute
    self:Unreal Editor Subsystem Object Reference=Get UnrealEditorSubsystem()
    CameraLocation:Vector="39.8744,60.689,177.1587"
    CameraRotation:Rotator="-18.7995,-121.9866,0"
   ) -> (then)
   Event BeginPlay()
   Event ActorBeginOverlap()
   Event Tick!() -> (OutputDelegate, then, DeltaSeconds:float)
  ]}
 UserConstructionScript {nodes [
   Construction Script()
   > Set Timer by Event(Delegate=Create Event(), Time=0.1, bLooping=false)
  ]}
]
variables []
