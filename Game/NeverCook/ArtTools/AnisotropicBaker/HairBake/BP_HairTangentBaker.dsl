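; Art-tool actor that bakes a hair strand map into an anisotropic tangent texture and previews it on a mesh.
; The construction script creates the bake and preview material instances, feeds in their settings, and applies the selected visualization to the static mesh.
; The Bake event renders tangents to a render target, repeatedly dilates them by ping-ponging between two targets, then blurs the result into the final bake target.
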(event EventBeginPlay)

(event Collision|EventActorBeginOverlap
  (params (OtherActor :type Actor%20Object%20Reference)))

(event EventTick
  (params (DeltaSeconds :type Float%20%28single-precision%29)))

(event Custom|Bake
  ;[1 1 1] Calculate Tangents
  (Rendering|Material|SetScalarParameterValue (Variables|Default|GetM_AnisotropyBaking) "bake" 1.0 :!class /Script/Engine.MaterialInstanceDynamic)
  (Rendering|DrawMaterialtoRenderTarget 0 (Variables|Default|GetAnisotropicRenderTarget) (Variables|Default|GetM_AnisotropyBaking))
  (Rendering|Material|SetScalarParameterValue (Variables|Default|GetM_AnisotropyBaking) "bake" 0.0 :!class /Script/Engine.MaterialInstanceDynamic)
  ;[1 1 1] Setup dilation
  (Rendering|Material|SetTextureParameterValue (Variables|Default|GetM_WriteToRenderTarget) "Texture" (Variables|Default|GetAnisotropicRenderTarget))
  (Rendering|DrawMaterialtoRenderTarget 0 (Variables|Default|GetDilateRenderTarget01) (Variables|Default|GetM_WriteToRenderTarget))
  (Rendering|DrawMaterialtoRenderTarget 0 (Variables|Default|GetDilateRenderTarget02) (Variables|Default|GetM_WriteToRenderTarget))
  (for _ (range 65)
    ;[1 1 1] Dilate texture
    (Rendering|Material|SetTextureParameterValue (Variables|Default|GetM_QuickDilate) "Texture" (Variables|Default|GetDilateRenderTarget01))
    (Rendering|DrawMaterialtoRenderTarget 0 (Variables|Default|GetDilateRenderTarget02) (Variables|Default|GetM_QuickDilate))
    (Rendering|Material|SetTextureParameterValue (Variables|Default|GetM_QuickDilate) "Texture" (Variables|Default|GetDilateRenderTarget02))
    (Rendering|DrawMaterialtoRenderTarget 0 (Variables|Default|GetDilateRenderTarget01) (Variables|Default|GetM_QuickDilate)))
  ;[1 1 1] Blur Tangents
  (Rendering|ClearRenderTarget2D 0 (Variables|Default|GetFinalBake(DoNotEdit)))
  (Rendering|Material|SetTextureParameterValue (Variables|Default|GetM_BlurBaking) "Texture" (Variables|Default|GetDilateRenderTarget01))
  (Rendering|DrawMaterialtoRenderTarget 0 (Variables|Default|GetFinalBake(DoNotEdit)) (Variables|Default|GetM_BlurBaking))
  (Rendering|ClearRenderTarget2D 0 (Variables|Default|GetDilateRenderTarget01))
  (Rendering|ClearRenderTarget2D 0 (Variables|Default|GetDilateRenderTarget02)))

(fn ConstructionScript
  (bind _m_strandvis (Variables|Default|GetM_StrandVis))
  (bind _m_anisotropybaking (Variables|Default|GetM_AnisotropyBaking))
  (bind _strandmap (Variables|Default|GetStrandMap))
  (Components|StaticMesh|SetStaticMesh (Variables|Default|GetStaticMesh) (Variables|Visualization|GetStaticMeshToVisualize))
  (Components|SkinnedMesh|SetSkinnedAssetandUpdate (Variables|Default|GetSkeletalMesh) (Variables|Visualization|GetSkelMeshToVisualize) true)
  (bind _returnvalue (Rendering|Material|CreateDynamicMaterialInstance 0 "/Game/NeverCook/ArtTools/AnisotropicBaker/HairBake/BakeMaterials/M_WriteToRenderTarget.M_WriteToRenderTarget" "None" "None" :!class /Script/Engine.KismetMaterialLibrary))
  (Variables|Default|SetM_WriteToRenderTarget _returnvalue)
  (bind _returnvalue_2 (Rendering|Material|CreateDynamicMaterialInstance 0 "/Game/NeverCook/ArtTools/AnisotropicBaker/HairBake/BakeMaterials/M_QuickDilate.M_QuickDilate" "None" "None" :!class /Script/Engine.KismetMaterialLibrary))
  (Variables|Default|SetM_QuickDilate _returnvalue_2)
  (bind _returnvalue_3 (Rendering|Material|CreateDynamicMaterialInstance 0 "/Game/NeverCook/ArtTools/AnisotropicBaker/HairBake/VisualizingMaterials/M_Visualize_Strandmap.M_Visualize_Strandmap" "None" "None" :!class /Script/Engine.KismetMaterialLibrary))
  (Variables|Default|SetM_StrandVis _returnvalue_3)
  (bind _returnvalue_4 (Rendering|Material|CreateDynamicMaterialInstance 0 "/Game/NeverCook/ArtTools/AnisotropicBaker/HairBake/VisualizingMaterials/M_VisualizeFinalBake.M_VisualizeFinalBake" "None" "None" :!class /Script/Engine.KismetMaterialLibrary))
  (Variables|Default|SetM_FinalTangentVis _returnvalue_4)
  (bind _returnvalue_5 (Rendering|Material|CreateDynamicMaterialInstance 0 "/Game/NeverCook/ArtTools/AnisotropicBaker/HairBake/BakeMaterials/M_HairAnisoBake.M_HairAnisoBake" "None" "None" :!class /Script/Engine.KismetMaterialLibrary))
  (Variables|Default|SetM_AnisotropyBaking _returnvalue_5)
  (bind _returnvalue_6 (Rendering|Material|CreateDynamicMaterialInstance 0 "/Game/NeverCook/ArtTools/AnisotropicBaker/HairBake/BakeMaterials/M_BlurBake.M_BlurBake" "None" "None" :!class /Script/Engine.KismetMaterialLibrary))
  (Variables|Default|SetM_BlurBaking _returnvalue_6)
  (Rendering|Material|SetScalarParameterValue (Variables|Default|GetM_BlurBaking) "BlurRadius" (Variables|Default|GetBlurRadius) :!class /Script/Engine.MaterialInstanceDynamic)
  (bind _returnvalue_7 (Rendering|Material|CreateDynamicMaterialInstance 0 "/Game/NeverCook/ArtTools/AnisotropicBaker/HairBake/VisualizingMaterials/M_Visualize_TangentsAsWorld.M_Visualize_TangentsAsWorld" "None" "None" :!class /Script/Engine.KismetMaterialLibrary))
  (Variables|Default|SetM_TangentAsWorldVis _returnvalue_7)
  (Rendering|Material|SetTextureParameterValue _m_anisotropybaking "StrandMap" _strandmap)
  (bind _size (Class|VolumetricsBPLIbrary|GetTextureSize _strandmap))
  (Rendering|Material|SetScalarParameterValue _m_anisotropybaking "ImageSize" (Math|Conversions|ToFloat(Integer) (Utilities|Struct|BreakIntPoint _size)) :!class /Script/Engine.MaterialInstanceDynamic)
  (Rendering|Material|SetScalarParameterValue (Variables|Default|GetM_AnisotropyBaking) "Method02" (Math|Conversions|ToFloat(Boolean) (Variables|Default|GetTangentMethod02)) :!class /Script/Engine.MaterialInstanceDynamic)
  (switch Utilities|FlowControl|Switch|SwitchonAnisoVisualization (Variables|Visualization|GetVisualization)
    (:NewEnumerator0
      (Rendering|Material|SetTextureParameterValue _m_strandvis "Texture" (Variables|Default|GetStrandMap))
      (Rendering|Material|SetMaterial (Variables|Default|GetStaticMesh) (Variables|Visualization|GetMat%20ID%20Element) _m_strandvis))
    (:NewEnumerator2
      (Rendering|Material|SetMaterial (Variables|Default|GetStaticMesh) (Variables|Visualization|GetMat%20ID%20Element) (Variables|Default|GetM_AnisotropyBaking))
      (Rendering|Material|SetTextureParameterValue (Variables|Default|GetM_AnisotropyBaking) "Normal" (Variables|Visualization|GetNormalMap)))
    (:NewEnumerator3
      (Rendering|Material|SetTextureParameterValue (Variables|Default|GetM_FinalTangentVis) "Normal" (Variables|Visualization|GetNormalMap))
      (Rendering|Material|SetTextureParameterValue (Variables|Default|GetM_FinalTangentVis) "Tangent" (Variables|Default|GetFinalBake(DoNotEdit)))
      (Rendering|Material|SetMaterial (Variables|Default|GetStaticMesh) (Variables|Visualization|GetMat%20ID%20Element) (Variables|Default|GetM_FinalTangentVis)))
    (:NewEnumerator4
      (Rendering|Material|SetTextureParameterValue (Variables|Default|GetM_TangentAsWorldVis) "Tangent" (Variables|Default|GetFinalBake(DoNotEdit)))
      (Rendering|Material|SetMaterial (Variables|Default|GetStaticMesh) (Variables|Visualization|GetMat%20ID%20Element) (Variables|Default|GetM_TangentAsWorldVis)))))
